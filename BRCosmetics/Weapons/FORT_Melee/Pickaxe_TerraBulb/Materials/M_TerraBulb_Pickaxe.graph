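# Material for the TerraBulb pickaxe that adds an animated glowing leaf effect to a masked area via the 2023 character-parent PreFX/PostFX functions.
# The masked area blends to a solid-colored subsurface material with added emissive.
# Two panning, parallaxed leaf textures supply normals for fake lit, scattering and Fresnel shading from the directional light direction.
# Three scrolling, hash-randomized flipbook layers, projected from local-space coordinates with bump offset, form animated leaves.
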Material /BRCosmetics/Weapons/FORT_Melee/Pickaxe_TerraBulb/Materials/M_TerraBulb_Pickaxe
graph {
 properties {
  bAutomaticallySetUsageInEditor=false
  bUseMaterialAttributes=true
  bUsedWithMeshDeformer=true
  bUsedWithMorphTargets=true
  bUsedWithSkeletalMesh=true
  pixelDepthOffsetMode=PDOM_Legacy
  shadingModel=MSM_FromMaterialExpression
  usedShadingModels=Default Lit | Subsurface | Clear Coat | Subsurface Profile | Cloth
 }
 nodes [
  n1 = Constant() {r=0.2}
  n2 = Constant() {r=8}
  n3 = Multiply(A=Multiply(A=Constant2Vector() {g=0.1}, B=@n2), B=Time())
  n4 = ComponentMask(None=Multiply(A=NamedRerouteUsage(), B=Constant3Vector() {constant=<rgba 1 1 -1 0>})) {b=true, r=true}
  n5 = Divide(A=@n4) {constB=5}
  n6 = MaterialFunctionCall!(Angle=Constant() {r=0.1}, UV/2DVector=Add(A=@n5, B=Constant2Vector() {g=0.2, r=0.1})) {materialFunction=/Game/Characters/Player/Male/Medium/Bodies/M_MED_Faux/Materials/MF_2DRotate}
  n7 = BumpOffset(Coordinate=@n6.Result, Height=Constant() {r=-1})
  n8 = Add(A=@n7, B=@n3)
  n9 = Floor(None=@n8)
  n10 = Saturate(None=ComponentMask(None=MaterialFunctionCall(Seed=@n1, Vec2=@n9) {materialFunction=&MF_Hash}.Vec2) {g=true})
  LinearInterpolate(Alpha=@n10)
  LocalPosition()
  n11 = MaterialFunctionCall() {materialFunction=&MF_CharacterParent_2023_PreFX}
  n12 = GetMaterialAttributes(None=@n11.Result) {attributeGetTypes [#1, #2]}
  n13 = VectorParameter() {parameterName=SolidBladeColor, defaultValue=<rgba 0.0111 0.0948 0.0594 1>}
  n14 = SetMaterialAttributes!(
   MaterialAttributes=@n12.MaterialAttributes
   Emissive Color=Add(A=@n12.Emissive Color, B=NamedRerouteUsage())
   Base Color=@n13.RGB
   Shading Model=ShadingModel() {shadingModel=MSM_Subsurface}
   Fuzz Color or Subsurface Color=@n13.RGB
   Opacity=Constant() {r=0.5}
   Metallic=Constant() {r=0.8}
  ) {attributeSetTypes [
    #1
    #3
    #4
    #5
    #6
    #7
   ]}
  MaterialFunctionCall(
   Material=BlendMaterialAttributes(A=@n12.MaterialAttributes, B=@n14, Alpha=NamedRerouteUsage())
   M=@n11.M
   Pre Skinned Local Position=@n11.M
   Distance Blend=@n11.M
   Fresnel=@n11.M
  ) {materialFunction=&MF_CharacterParent_2023_PostFX}
  NamedRerouteDeclaration(None=StaticComponentMaskParameter(None=@n11.Tech Art Mask) {defaultR=true, parameterName=LeafChannel}) {name=leaf_Area, nodeColor=<rgba 0 0.2969 1 1>}
  n15 = VectorParameter() {group=Leaf, parameterName=LeafLit, defaultValue=<rgba 0.3094 1 0.7996 1>}
  n16 = Multiply(A=@n15.RGB, B=@n15.RGB)
  NamedRerouteDeclaration(None=@n16) {name=LitColor, nodeColor=<rgba 1 0.8906 0 1>}
  n17 = Custom!(None) {code="return View.DirectionalLightDirection;", showCode=true}
  n18 = Transform(None=NamedRerouteUsage())
  NamedRerouteDeclaration(None=Saturate(None=DotProduct(A=@n17, B=@n18))) {name=Lit_Surface, nodeColor=<rgba 1 0.3984 0 1>}
  NamedRerouteDeclaration(None=@n11.Pre skinned Local Position) {name=PreskinnedPOS, nodeColor=<rgba 1 0 0.9141 1>}
  NamedRerouteDeclaration(None=@n12.Normal) {name=Normal_IN, nodeColor=<rgba 0 0.5547 1 1>}
  n19 = Multiply(A=@n17) {constB=-1}
  NamedRerouteDeclaration(None=Saturate(None=DotProduct(A=@n19, B=@n18))) {name=SSS_Surface, nodeColor=<rgba 0.875 1 0 1>}
  n20 = VectorParameter() {parameterName=LeafUnLit, defaultValue=<rgba 0.0272 0.2323 0.1456 1>}
  n21 = Multiply(A=@n20.RGB, B=@n20.RGB)
  n22 = Divide(A=@n4, B=Constant() {r=10})
  n23 = Constant() {r=-3}
  n24 = TextureSample(UVs=Panner(Coordinate=BumpOffset(Coordinate=@n22, Height=@n23)) {speedY=0.1}) {samplerType=SAMPLERTYPE_LinearColor, texture=/BRCosmetics/Weapons/FORT_Melee/Pickaxe_TerraBulb/Textures/T_leaves_01}
  n25 = Divide(A=@n4, B=Constant() {r=20})
  n26 = Constant() {r=-1}
  n27 = TextureSample(UVs=Panner(Coordinate=BumpOffset(Coordinate=@n25, Height=@n26)) {speedY=0.1}) {samplerType=SAMPLERTYPE_LinearColor, texture=/BRCosmetics/Weapons/FORT_Melee/Pickaxe_TerraBulb/Textures/T_leaves_02}
  n28 = LinearInterpolate(A=Multiply(A=@n24.RGB, B=Multiply(A=@n24.B) {constB=0.5}), B=Multiply(A=@n27.RGB, B=@n27.RGB), Alpha=@n27.B)
  n29 = Subtract(A=ComponentMask(None=@n28) {g=true, r=true}) {constB=0.5}
  n30 = Transform(None=Normalize(VectorInput=DeriveNormalZ(InXY=@n29)))
  n31 = DotProduct(A=@n17, B=@n30)
  n32 = SmoothStep(Value=Power(Base=Saturate(None=@n31)) {constExponent=3})
  n33 = Fresnel(Normal=@n30) {baseReflectFraction=0, exponent=1}
  NamedRerouteDeclaration(None=Add(A=LinearInterpolate(A=@n21, B=@n16, Alpha=@n32), B=Multiply(A=@n33) {constB=0.2})) {name=Lit, nodeColor=<rgba 0 1 0.2734 1>}
  n34 = Saturate(None=DotProduct(A=@n30, B=@n19))
  n35 = VectorParameter() {group=Leaf, parameterName=LeafSSS, defaultValue=<rgba 0 0.5531 0.1684 2>}
  NamedRerouteDeclaration(None=Multiply(A=Power(Base=@n34) {constExponent=3}, B=Multiply(A=@n35.RGB, B=@n35.RGB))) {name=SSS, nodeColor=<rgba 0 1 0.6719 1>}
  n36 = NamedRerouteUsage()
  n37 = Multiply(A=NamedRerouteUsage()) {constB=0.8}
  n38 = VectorParameter() {group=Leaf, parameterName=SurfaceLit, defaultValue=<rgba 0 0.4531 0.3022 1>}
  n39 = Multiply(A=Power(Base=NamedRerouteUsage()), B=Multiply(A=@n38.RGB, B=@n38.RGB))
  n40 = Multiply(A=Add(A=NamedRerouteUsage(), B=NamedRerouteUsage()), B=ComponentMask(None=@n28) {b=true})
  n41 = Multiply(A=TextureSample() {texture=/BRCosmetics/Weapons/FORT_Melee/Pickaxe_TerraBulb/Textures/T_TerraBulb_Pickaxe_TechArt}.R, B=NamedRerouteUsage())
  n42 = Constant() {r=0.6}
  n43 = NamedRerouteUsage()
  n44 = VectorParameter() {group=Leaf, parameterName=SurfaceSSS, defaultValue=<rgba 0 0.4698 0.1513 1>}
  n45 = Add(
   A=Add(A=Add(A=@n39, B=@n40), B=Multiply(A=@n41, B=@n42))
   B=Multiply(A=Power(Base=@n43) {constExponent=5}, B=Multiply(A=@n44.RGB, B=@n44.RGB))
  )
  n46 = Multiply(A=Power(Base=NamedRerouteUsage()) {constExponent=4}, B=Multiply(A=NamedRerouteUsage()) {constB=0.5})
  NamedRerouteDeclaration(None=Add(A=Multiply(A=@n36, B=@n37), B=Add(A=@n45, B=@n46))) {name=EM_OUT, nodeColor=<rgba 0.8047 0 1 1>}
  n47 = ReflectionVectorWS(CustomWorldNormal=@n18)
  NamedRerouteDeclaration(None=Saturate(None=DotProduct(A=@n17, B=@n47))) {name=Highlight_Surface, nodeColor=<rgba 1 0.2344 0 1>}
  n48 = Constant() {r=8}
  n49 = Multiply(A=Multiply(A=Constant2Vector() {g=0.12}, B=@n48), B=Time())
  n50 = BumpOffset(Coordinate=@n5, Height=Constant() {r=-1})
  n51 = Add(A=@n50, B=@n49)
  n52 = Frac(None=@n51)
  n53 = Constant2Vector() {g=0.5, r=0.5}
  n54 = Saturate(None=SmoothStep(Value=Distance(A=@n52, B=@n53)) {constMax=0.47, constMin=0.42})
  n55 = Add(A=@n50, B=@n49)
  n56 = Frac(None=Divide(A=Floor(None=@n55), B=Constant() {r=16}))
  n57 = Add(A=ComponentMask(None=@n56) {r=true}, B=ComponentMask(None=@n56) {g=true})
  n58 = Add(A=Multiply(A=Sine(None=@n57)) {constB=2.2}, B=Multiply(A=Time()) {constB=0.4})
  n59 = TextureSample(
   UVs=MaterialFunctionCall("Animation  Phase (0-1)"=Frac(None=@n58), Number of Rows=@n48, Number of Columns=@n48, UVs=Frac(None=@n55)) {materialFunction=&FlipBook}.UVs
   Tex=TextureObject() {samplerType=SAMPLERTYPE_LinearColor, texture=/BRCosmetics/Weapons/FORT_Melee/Pickaxe_TerraBulb/Textures/T_LeafFlipbook}
  ) {
   samplerSource=SSM_Wrap_WorldGroupSettings
   samplerType=SAMPLERTYPE_LinearColor
   texture=&T_LeafFlipbook
  }
  n60 = MaterialFunctionCall(Vec2=Floor(None=@n51)) {materialFunction=&MF_Hash}
  n61 = Multiply(A=Multiply(A=OneMinus(None=@n54), B=@n59.R), B=Saturate(None=ComponentMask(None=@n60.Vec2) {g=true}))
  n62 = Distance(A=Frac(None=@n8), B=Constant2Vector() {g=0.5, r=0.5})
  n63 = OneMinus(None=Saturate(None=SmoothStep(Value=@n62) {constMax=0.47, constMin=0.42}))
  n64 = Add(A=@n7, B=@n3)
  n65 = Frac(None=Divide(A=Floor(None=@n64), B=Constant() {r=16}))
  n66 = Add(A=ComponentMask(None=@n65) {r=true}, B=ComponentMask(None=@n65) {g=true})
  n67 = Time()
  n68 = Add(A=Multiply(A=Sine(None=@n66)) {constB=1.7}, B=Add(A=Multiply(A=@n67) {constB=0.45}) {constB=0.2})
  n69 = TextureSample(
   UVs=MaterialFunctionCall("Animation  Phase (0-1)"=Frac(None=@n68), Number of Rows=@n2, Number of Columns=@n2, UVs=Frac(None=@n64)) {materialFunction=&FlipBook}.UVs
   Tex=TextureObject() {samplerType=SAMPLERTYPE_LinearColor, texture=&T_LeafFlipbook}
  ) {
   samplerSource=SSM_Wrap_WorldGroupSettings
   samplerType=SAMPLERTYPE_LinearColor
   texture=&T_LeafFlipbook
  }
  n70 = Constant() {r=8}
  n71 = Multiply(A=Multiply(A=Constant2Vector() {g=0.15}, B=@n70), B=Time())
  n72 = Constant2Vector() {g=0.2, r=-0.2}
  n73 = BumpOffset(
   Coordinate=MaterialFunctionCall(Angle=Constant() {r=-0.08}, UV/2DVector=Add(A=@n6.Result, B=@n72)) {materialFunction=&MF_2DRotate}.Result
   Height=Constant() {r=-1}
  )
  n74 = Add(A=@n73, B=@n71)
  n75 = Distance(A=Frac(None=@n74), B=Constant2Vector() {g=0.5, r=0.5})
  n76 = OneMinus(None=Saturate(None=SmoothStep(Value=@n75) {constMax=0.47, constMin=0.42}))
  n77 = Add(A=@n73, B=@n71)
  n78 = Frac(None=Divide(A=Floor(None=@n77), B=Constant() {r=16}))
  n79 = Add(A=ComponentMask(None=@n78) {r=true}, B=ComponentMask(None=@n78) {g=true})
  n80 = Time()
  n81 = Add(A=Multiply(A=Sine(None=@n79)) {constB=1.3}, B=Add(A=Multiply(A=@n80) {constB=0.334}) {constB=0.5})
  n82 = TextureSample(
   UVs=MaterialFunctionCall("Animation  Phase (0-1)"=Frac(None=@n81), Number of Rows=@n70, Number of Columns=@n70, UVs=Frac(None=@n77)) {materialFunction=&FlipBook}.UVs
   Tex=TextureObject() {samplerType=SAMPLERTYPE_LinearColor, texture=&T_LeafFlipbook}
  ) {
   samplerSource=SSM_Wrap_WorldGroupSettings
   samplerType=SAMPLERTYPE_LinearColor
   texture=&T_LeafFlipbook
  }
  n83 = ComponentMask(None=MaterialFunctionCall(Seed=Constant() {r=0.12}, Vec2=Floor(None=@n74)) {materialFunction=&MF_Hash}.Vec2) {g=true}
  n84 = Add(A=Multiply(A=Multiply(A=@n63, B=@n69.R), B=@n10), B=Multiply(A=Multiply(A=@n76, B=@n82.R), B=Saturate(None=@n83)))
  NamedRerouteDeclaration(None=Saturate(None=Add(A=@n61, B=@n84))) {name=AnimatedLeaf, nodeColor=<rgba 0 1 0.7656 1>}
 ]
}
